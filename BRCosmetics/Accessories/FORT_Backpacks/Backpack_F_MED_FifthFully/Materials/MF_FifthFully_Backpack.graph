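# Material function for the FifthFully backpack that layers animated fire and shimmer onto incoming material attributes, overriding BaseColor and EmissiveColor.
# Three packed fire-texture channels get scaled, panning UVs, normal-based distortion and three-color gradients, with per-side wing offsets by pre-skinned local X for red and green.
# View-direction and BumpOffset lookups supply shimmer, while input masks control blending.
# Several iridescence and curve-atlas branches end unconnected.
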MaterialFunction /BRCosmetics/Accessories/FORT_Backpacks/Backpack_F_MED_FifthFully/Materials/MF_FifthFully_Backpack
graph {nodes [
  n1 = Frac(None=Multiply(A=Time()) {constB=0.1})
  n2 = Add(A=Multiply(A=TextureCoordinate()) {constB=0.5}, B=ConstantBiasScale(None=Sine(None=@n1)))
  n3 = Add(A=ComponentMask(None=@n2) {r=true}, B=ComponentMask(None=@n2) {g=true})
  CurveAtlasRowParameter(CurveTime=Add(A=Divide(A=@n3)) {constB=-0.048}) {
   atlas=&AthenaCurveAtlas01
   curve=/Game/Characters/Player/Male/Medium/Bodies/M_MED_HighTower_Grape/Textures/GrapeFoilSilverCurve
   defaultValue=245
   parameterName=BaseGradientCurve
  }
  n4 = GetMaterialAttributes(None=NamedRerouteUsage()) {attributeGetTypes [#1, #2, #3]}
  n5 = LinearInterpolate(
   A=NamedRerouteDeclaration(None=@n4.EmissiveColor) {name=In_EmissiveColor, nodeColor=<rgba 0 0.1344 1 1>}
   B=NamedRerouteUsage()
   Alpha=Multiply(A=NamedRerouteUsage(), B=ScalarParameter() {defaultValue=0.333, parameterName=FireBaseColorAmountMultiplier})
  )
  n6 = NamedRerouteUsage()
  n7 = ScalarParameter() {defaultValue=0.25, parameterName=ShimmerColorAMultiplier}
  n8 = NamedRerouteUsage()
  n9 = Multiply(A=Add(A=Multiply(A=@n6, B=@n7), B=Multiply(B=@n8) {constA=1}), B=OneMinus(None=NamedRerouteUsage()))
  n10 = LinearInterpolate(A=@n5, B=Saturate(None=Multiply(A=@n5, B=@n9)), Alpha=OneMinus(None=NamedRerouteUsage()))
  n11 = NamedRerouteUsage()
  FunctionOutput(None=SetMaterialAttributes!(MaterialAttributes=@n4.MaterialAttributes, BaseColor=Add(A=@n10, B=@n11), EmissiveColor=NamedRerouteUsage()) {attributeSetTypes [#1, #3]})
  n12 = DotProduct(A=CameraVectorWS(), B=Transform(None=NamedRerouteUsage()))
  n13 = ScalarParameter() {defaultValue=1, parameterName=IrridescentAngleMultiplier}
  n14 = CurveAtlasRowParameter(CurveTime=Saturate(None=Multiply(A=@n12, B=@n13))) {
   atlas=/Game/Characters/Shared/CurveAtlas/AthenaCharacterCurveAtlas
   curve=/Game/Packages/Fortress_SharedMaterials/Anime_Master_Material/AnisoTropicHighLightCurve
   defaultValue=15
  }
  n15 = ScalarParameter() {defaultValue=0.1, parameterName=IrredescentMultiplier}
  LinearInterpolate(B=@n14, Alpha=@n15)
  n16 = NamedRerouteUsage()
  Multiply(A=@n15, B=Add(A=ComponentMask(None=@n16) {g=true}, B=ComponentMask(None=@n16) {b=true}))
  NamedRerouteDeclaration(None=FunctionInput(Preview=TextureSample() {texture=&T_BlacksRGB}.RGB) {inputName=In_TechArtMask, sortPriority=1}) {name=In_TechArtMask, nodeColor=<rgba 0.0312 0 1 1>}
  NamedRerouteDeclaration(None=FunctionInput() {inputName=In_MaterialAttributes, inputType=FunctionInput_MaterialAttributes}) {name=In_MaterialAttributes, nodeColor=<rgba 0.2891 1 0 1>}
  n17 = ComponentMask(None=VectorParameter() {parameterName=RedChannelUVScaleAndSpeed, defaultValue=<rgba 2 1 0 0.1>}.RGBA) {a=true, b=true}
  n18 = Multiply(A=Constant2Vector() {g=-1, r=1}, B=@n17)
  n19 = ComponentMask(None=NamedRerouteUsage()) {b=true}
  NamedRerouteDeclaration(None=Multiply(A=Time(), B=LinearInterpolate(A=@n18, B=@n17, Alpha=@n19)) {constB=0.1}) {name=RedChannelUVAnimation, nodeColor=<rgba 1 0.7266 0 1>}
  n20 = Constant3Vector() {constant=<rgba 0 0 1 0>}
  n21 = ScalarParameter() {defaultValue=1, parameterName=ViewDirectionMultiplier}
  NamedRerouteDeclaration(None=Multiply(A=Transform(None=@n20) {transformSourceType=TRANSFORMSOURCE_View}, B=Abs(None=@n21))) {name=ViewDirectionForShimmer, nodeColor=<rgba 1 0 0.2812 1>}
  NamedRerouteDeclaration(None=FunctionInput(Preview=TextureSample() {texture=&T_BlacksRGB}.RGB) {inputName=In_FireTipMask, sortPriority=2}) {name=In_FireShadingMask, nodeColor=<rgba 0.0312 0 1 1>}
  n22 = VectorParameter() {parameterName=GreenChannelUVScaleAndSpeed, defaultValue=<rgba 1 1 0 0.12>}
  n23 = Add(A=Multiply(A=NamedRerouteUsage(), B=ComponentMask(None=@n22.RGBA) {g=true, r=true}), B=NamedRerouteUsage())
  n24 = NamedRerouteUsage()
  NamedRerouteDeclaration(None=Add(A=Add(A=@n23, B=@n24), B=NamedRerouteUsage())) {name=GreenChannelUVCoords, nodeColor=<rgba 0.8353 0.067 1 1>}
  n25 = ComponentMask(None=VectorParameter() {parameterName=GreenChannelUVScaleAndSpeed, defaultValue=<rgba 1 1 0 0.12>}.RGBA) {a=true, b=true}
  n26 = Multiply(A=Constant2Vector() {g=-1, r=1}, B=@n25)
  n27 = ComponentMask(None=NamedRerouteUsage()) {b=true}
  NamedRerouteDeclaration(None=Multiply(A=Time(), B=LinearInterpolate(A=@n26, B=@n25, Alpha=@n27)) {constB=0.1}) {name=GreenChannelUVAnimation, nodeColor=<rgba 1 0.7266 0 1>}
  n28 = NamedRerouteUsage()
  NamedRerouteDeclaration(None=Add(A=Transform(None=@n28), B=NamedRerouteUsage())) {name=WarpedViewDirection, nodeColor=<rgba 0 0.6016 1 1>}
  n29 = Multiply(
   A=NamedRerouteUsage()
   B=ComponentMask(None=VectorParameter() {parameterName=BlueChannelUVScaleAndSpeed, defaultValue=<rgba 1 1 0 0.133>}.RGBA) {g=true, r=true}
  )
  n30 = NamedRerouteUsage()
  NamedRerouteDeclaration(None=Add(A=Add(A=@n29, B=@n30), B=NamedRerouteUsage())) {name=BlueChannelUVCoords, nodeColor=<rgba 0 0.0368 1 1>}
  n31 = ComponentMask(None=VectorParameter() {parameterName=BlueChannelUVScaleAndSpeed, defaultValue=<rgba 1 1 0 0.133>}.RGBA) {a=true, b=true}
  n32 = Multiply(A=Constant2Vector() {g=-1, r=1}, B=@n31)
  n33 = ComponentMask(None=NamedRerouteUsage()) {b=true}
  NamedRerouteDeclaration(None=Multiply(A=Time(), B=LinearInterpolate(A=@n32, B=@n31, Alpha=@n33)) {constB=0.1}) {name=BlueChannelUVAnimation, nodeColor=<rgba 0 0.1272 1 1>}
  n34 = TextureSample(UVs=NamedRerouteUsage(), Tex=NamedRerouteUsage()) {samplerType=SAMPLERTYPE_Masks, texture=/BRCosmetics/Accessories/FORT_Backpacks/Backpack_F_MED_FifthFully/Textures/T_FifthFully_Backpack_FX}
  NamedRerouteDeclaration(None=@n34.G) {name=PackedFireTextureGreenChannel, nodeColor=<rgba 0 1 0.0251 1>}
  NamedRerouteDeclaration(None=@n4.BaseColor) {name=In_BaseColor, nodeColor=<rgba 1 0 0.6094 1>}
  n35 = TextureSample(UVs=NamedRerouteUsage(), Tex=NamedRerouteUsage()) {samplerType=SAMPLERTYPE_Masks, texture=&T_FifthFully_Backpack_FX}
  NamedRerouteDeclaration(None=@n35.B) {name=PackedFireTextureBlueChannel, nodeColor=<rgba 0 0 1 1>}
  n36 = ComponentMask(None=NamedRerouteUsage()) {g=true, r=true}
  n37 = Constant() {r=2}
  NamedRerouteDeclaration(None=TextureSampleParameter2D(UVs=BumpOffset(Coordinate=@n36, Height=@n37)) {parameterName=ShimmerColorTexture, texture=&T_BlacksRGB}.RGB) {name=ShimmerColorA, nodeColor=<rgba 0 0.5312 1 1>}
  n38 = VectorParameter() {parameterName=RedChannelWingOffsets, defaultValue=<rgba 0 0.5 0 0>}
  n39 = ComponentMask(None=VertexInterpolator(VS=LocalPosition() {localOrigin=InstancePreSkinning}.XYZ).PS) {r=true}
  n40 = Constant() {r=1}
  n41 = Constant()
  NamedRerouteDeclaration(None=LinearInterpolate(
    A=ComponentMask(None=@n38.RGB) {g=true, r=true}
    B=ComponentMask(None=@n38.RGBA) {a=true, b=true}
    Alpha=If(A=@n39, "A > B"=@n40, "A == B"=@n41, "A < B"=@n41)
   )) {name=RedChannelWingOffsets, nodeColor=<rgba 0 0.8828 1 1>}
  n42 = VectorParameter() {parameterName=GreenChannelWingOffsets, defaultValue=<rgba 0 0.5 0 0>}
  n43 = ComponentMask(None=VertexInterpolator(VS=LocalPosition() {localOrigin=InstancePreSkinning}.XYZ).PS) {r=true}
  n44 = Constant() {r=1}
  n45 = Constant()
  NamedRerouteDeclaration(None=LinearInterpolate(
    A=ComponentMask(None=@n42.RGB) {g=true, r=true}
    B=ComponentMask(None=@n42.RGBA) {a=true, b=true}
    Alpha=If(A=@n43, "A > B"=@n44, "A == B"=@n45, "A < B"=@n45)
   )) {name=GreenChannelWingOffsets, nodeColor=<rgba 0 0.8828 1 1>}
  NamedRerouteDeclaration(None=MaterialFunctionCall(
    A=VectorParameter() {parameterName=RedChannelColorA, defaultValue=<rgba 0.1559 0.0452 0.117 1>}.RGB
    B=VectorParameter() {parameterName=RedChannelColorB, defaultValue=<rgba 0.2435 0.1015 0.0922 1>}.RGB
    C=VectorParameter() {parameterName=RedChannelColorC, defaultValue=<rgba 0.8388 0.491 0.1301 1>}.RGB
    Alpha=NamedRerouteUsage()
   ) {materialFunction=&Lerp_3Color}.Result) {name=RedChannelColors, nodeColor=<rgba 0.4766 0 1 1>}
  n46 = VectorParameter() {parameterName=A}
  n47 = VectorParameter() {parameterName=B, defaultValue=<rgba 0.8228 0.2918 0.1946 1>}
  n48 = VectorParameter() {parameterName=C, defaultValue=<rgba 0.552 0.5089 0.162 1>}
  n49 = Add(A=TextureCoordinate(), B=Multiply(A=Time(), B=Constant2Vector() {r=-0.1}))
  n50 = ScalarParameter() {parameterName=ShimmerBumpHeight}
  n51 = TextureSampleParameter2D(UVs=NamedRerouteDeclaration(None=BumpOffset(Coordinate=@n49, Height=@n50)) {name=BumpOffsetUV, nodeColor=<rgba 0.6406 0 1 1>}) {parameterName=BumpOffsetShimmerTexture, texture=&T_BlacksRGB}
  NamedRerouteDeclaration(None=Multiply(
    A=MaterialFunctionCall(A=@n46.RGB, B=@n47.RGB, C=@n48.RGB, Alpha=@n51.R) {materialFunction=&Lerp_3Color}.Result
    B=ScalarParameter() {defaultValue=1, parameterName=BumpOffsetShimmerMultiplier}
   )) {name=FireOnWings, nodeColor=<rgba 0 1 0.9297 1>}
  NamedRerouteDeclaration(None=MaterialFunctionCall(
    A=VectorParameter() {parameterName=GreenChannelColorA, defaultValue=<rgba 0.2083 0.0152 0.0139 1>}.RGB
    B=VectorParameter() {parameterName=GreenChannelColorB, defaultValue=<rgba 0.9047 0.2016 0.0685 1>}.RGB
    C=VectorParameter() {parameterName=GreenChannelColorC, defaultValue=<rgba 0.9301 0.7913 0.1812 1>}.RGB
    Alpha=NamedRerouteUsage()
   ) {materialFunction=&Lerp_3Color}.Result) {name=GreenChannelColors, nodeColor=<rgba 1 0.6005 0 1>}
  NamedRerouteDeclaration(None=MaterialFunctionCall(
    A=VectorParameter() {parameterName=BlueChannelColorA, defaultValue=<rgba 0.1492 0.0595 0.1719 1>}.RGB
    B=VectorParameter() {parameterName=BlueChannelColorB, defaultValue=<rgba 0.4435 0.125 0.026 1>}.RGB
    C=VectorParameter() {parameterName=BlueChannelColorC, defaultValue=<rgba 0.6957 0.5108 0.121 1>}.RGB
    Alpha=NamedRerouteUsage()
   ) {materialFunction=&Lerp_3Color}.Result) {name=BlueChannelColors, nodeColor=<rgba 1 0.6005 0 1>}
  n52 = NamedRerouteUsage()
  n53 = NamedRerouteUsage()
  NamedRerouteDeclaration(None=Add(A=Multiply(A=@n52, B=@n53), B=NamedRerouteUsage())) {name=CombinedColors, nodeColor=<rgba 0 0.4844 1 1>}
  NamedRerouteDeclaration(None=@n4.Normal) {name=In_Normal, nodeColor=<rgba 0 1 0.3438 1>}
  n54 = TextureSampleParameter2D() {parameterName=BumpOffsetHeight, texture=&T_BlacksRGB}
  n55 = Constant() {r=1}
  n56 = TextureSample(
   UVs=BumpOffset(
    Coordinate=TextureCoordinate()
    Height=Multiply(A=@n54.R, B=@n55)
    HeightRatioInput=ScalarParameter() {defaultValue=0.05, parameterName=HeightRatioInput}
   )
   Tex=TextureObject() {texture=&T_BlacksRGB}
  ) {samplerType=SAMPLERTYPE_Masks, texture=/BRCosmetics/Accessories/FORT_Backpacks/Backpack_F_MED_FifthFully/Textures/T_FifthFully_Backpack_S}
  NamedRerouteDeclaration(None=@n56.RGB) {name=BumpOffsetHighlights, nodeColor=<rgba 0 1 0.0391 1>}
  n57 = ComponentMask(None=NamedRerouteUsage()) {g=true}
  n58 = NamedRerouteUsage()
  n59 = Saturate(None=Add(A=ComponentMask(None=@n58) {g=true}, B=ComponentMask(None=@n58) {b=true}))
  n60 = NamedRerouteDeclaration(None=Multiply(A=Saturate(None=@n57), B=NamedRerouteDeclaration(None=@n59) {name=CombinedTechArtMask, nodeColor=<rgba 0.3594 1 0 1>})) {name=CombinedMasks, nodeColor=<rgba 0 0.9062 1 1>}
  n61 = ScalarParameter() {defaultValue=1, parameterName=FireEmissiveAmountMultiplier}
  NamedRerouteDeclaration(None=LinearInterpolate(A=NamedRerouteUsage(), B=NamedRerouteUsage(), Alpha=Multiply(A=@n60, B=@n61))) {name=Out_Emissive, nodeColor=<rgba 0.6641 1 0 1>}
  n62 = ScalarParameter() {defaultValue=1, parameterName=NormalMapStrengthForShimmer}
  n63 = NamedRerouteUsage()
  n64 = ComponentMask(None=@n63) {g=true, r=true}
  NamedRerouteDeclaration(None=AppendVector(A=Multiply(A=@n62, B=@n64), B=ComponentMask(None=@n63) {b=true})) {name=NormalMapForShimmer, nodeColor=<rgba 0 1 0.0391 1>}
  NamedRerouteDeclaration(None=TextureObjectParameter() {parameterName=FireTexture, texture=&T_BlacksRGB}) {name=FireTexture, nodeColor=<rgba 0.0547 0 1 1>}
  n65 = MaterialFunctionCall(UVs=TextureCoordinate(), "Rotation Angle (0-1)"=ScalarParameter() {defaultValue=-0.25, parameterName=RotateTextureNormalized}) {materialFunction=&CustomRotator}
  n66 = VectorParameter() {parameterName=RedChannelUVScaleAndSpeed, defaultValue=<rgba 2 1 0 0.1>}
  n67 = Add(
   A=Multiply(
    A=NamedRerouteDeclaration(None=@n65.Rotated Values) {name=RotatedUVs, nodeColor=<rgba 0 1 0.0859 1>}
    B=ComponentMask(None=@n66.RGBA) {g=true, r=true}
   )
   B=NamedRerouteUsage()
  )
  n68 = NamedRerouteUsage()
  NamedRerouteDeclaration(None=Add(A=Add(A=@n67, B=@n68), B=NamedRerouteUsage())) {name=RedChannelUVCoords, nodeColor=<rgba 0.2656 1 0 1>}
  n69 = TextureSample(UVs=NamedRerouteUsage(), Tex=NamedRerouteUsage()) {samplerType=SAMPLERTYPE_Masks, texture=&T_FifthFully_Backpack_FX}
  NamedRerouteDeclaration(None=@n69.R) {name=PackedFireTextureRedChannel, nodeColor=<rgba 1 0 0.0251 1>}
  n70 = ConstantBiasScale(None=NamedRerouteUsage()) {bias=-0.5, scale=2}
  NamedRerouteDeclaration(None=Multiply(A=ComponentMask(None=@n70) {g=true, r=true}, B=ScalarParameter() {defaultValue=0.05, parameterName=UVDistortionNormalMultiplier})) {name=UVDistortion, nodeColor=<rgba 1 0.8438 0 1>}
  NamedRerouteUsage()
  NamedRerouteUsage()
  NamedRerouteUsage()
  SmoothStep(Value=@n56.RGB) {constMin=0.384}
 ]}
