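# Skeletal-mesh material for the FifthFully back bling (backpack) cosmetic, built on the shared 2023 character parent functions.
# PreFX material attributes pass through the backpack-specific MF_FifthFully_Backpack function, then into PostFX for final output.
# The backpack function also reads TechArtMask and FireTipsMask texture parameters, both defaulting to black.
# It uses material attributes with an expression-driven shading model and supports morph targets and mesh deformers.
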Material /BRCosmetics/Accessories/FORT_Backpacks/Backpack_F_MED_FifthFully/Materials/M_FifthFully_Backpack
graph {
 properties {
  bAutomaticallySetUsageInEditor=false
  bUseMaterialAttributes=true
  bUsedWithMeshDeformer=true
  bUsedWithMorphTargets=true
  bUsedWithSkeletalMesh=true
  pixelDepthOffsetMode=PDOM_Legacy
  previewMesh=/BRCosmetics/Accessories/FORT_Backpacks/Backpack_F_MED_FifthFully/Meshes/F_MED_FifthFully_Pack
  shadingModel=MSM_FromMaterialExpression
  usedShadingModels=Default Lit | Subsurface | Clear Coat | Subsurface Profile | Cloth
 }
 nodes [
  n1 = MaterialFunctionCall() {materialFunction=&MF_CharacterParent_2023_PreFX}
  MaterialFunctionCall(
   Material=MaterialFunctionCall!(
    In_MaterialAttributes=@n1.Result
    In_TechArtMask=TextureSampleParameter2D() {parameterName=TechArtMask, texture=&T_BlacksRGB}.RGB
    In_FireTipMask=TextureSampleParameter2D() {parameterName=FireTipsMask, texture=&T_BlacksRGB}.RGB
   ) {materialFunction=&/MF_FifthFully_Backpack}.Result
   M=@n1.M
   Pre Skinned Local Position=@n1.M
   Distance Blend=@n1.M
   Fresnel=@n1.M
  ) {materialFunction=&MF_CharacterParent_2023_PostFX}
 ]
}
